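# Niagara effect for the Shrimp Stroll emote that throws off a wintry spray of ice, snow and sparkles.
# It layers ice chunk meshes, air streak and powder sprites, and twin CPU and GPU snowflake emitters with twinkling sparkle values.
# A System.IsSynced flag switches spawn rates and tunes motion or opacity, steering streaks and flakes along System.SnowDirection when true.
# Time-of-day selectors and a shared color multiplier tint the sprites, boosting sparkle alpha at night.
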NiagaraSystem /BRCosmetics/Animation/Game/MainPlayer/Emotes/ShrimpStroll/FX/NS_Emote_ShrimpStroll
dependencies {
 usedDataInterfaces [
  &NiagaraDataInterfaceCurve
  &NiagaraDataInterfaceSpriteRendererInfo
  &NiagaraDataInterfaceMeshRendererInfo
  &NiagaraDataInterfaceVector2DCurve
  &NiagaraDataInterfaceCamera
  &NiagaraDataInterfaceVectorCurve
  &NiagaraDataInterfaceColorCurve
 ]
 usedDynamicInputs [
  &SetBoolByFloatComparison
  &LerpLinearColors
  &ColorFromCurve
 ]
 usedModules [
  &SystemState
  &self:SystemSpawnScript.SystemScriptSource.SystemScriptGraph.NiagaraNodeAssignment_0.SetVariables_#1
  &self:SystemSpawnScript.SystemScriptSource.SystemScriptGraph.NiagaraNodeAssignment_3.SetVariables_#2
  &self:SystemSpawnScript.SystemScriptSource.SystemScriptGraph.NiagaraNodeAssignment_2.SetVariables_#3
  &self:SystemSpawnScript.SystemScriptSource.SystemScriptGraph.NiagaraNodeAssignment_1.SetVariables_#4
 ]
}
emitters {
 IceChunks {
  emitterUpdateScript [
   &EmitterState()
   &SpawnRate(
    SpawnRate=&Multiply_Float(A=&MakeCustomFloatFromBool(Bool=$System.IsSynced, True Float=11, False Float=3), B=2.5)
    Use Spawn Probability=false
   )
  ]
  particleSpawnScript [
   &InitializeParticle(
    Color=<rgba 6.96 7.5 7.3919 1>
    Lifetime Max=1
    Lifetime Min=0.25
    Lifetime Mode=::NewEnumerator1
    Mesh Renderer Array Visibility Mode=::NewEnumerator1
    Mesh Renderer Info=DataInterface() {propertyValues {MeshRenderer=&self:IceChunks_0.NiagaraMeshRendererProperties_0}}
    Mesh Renderer Mesh Index:NiagaraInt32=&RandomRangeInt(Minimum=0, Maximum=2)
    Mesh Scale Mode=::NewEnumerator3
    Mesh Uniform Scale=&RandomRangeFloat(Minimum=0.1, Maximum=0.25)
    UsePositionOffset=false
    Write Material Random=true
   )
   &ShapeLocation(MaskSpawnGroup=false, Override Seed=false, Sphere Radius=11, UseOwnerScale=true)
   &AddVelocity(
    Cone Angle=&MakeCustomFloatFromBool(Bool=$System.IsSynced, True Float=180, False Float=180)
    Cone Axis=&MakeCustomVectorFromBool(Bool=$System.IsSynced, True Vector=<0 0.2 1>, False Vector=<0 -0.2 1>)
    Distribution Along Cone Axis=0.75
    Inner Cone Angle=5
    Module.UseConeFalloff=true
    Module.UseVelDistribution=true
    Override Seed=false
    Speed Falloff From Cone Axis=0.3
    Velocity Mode=::NewEnumerator2
    Velocity Speed=&MakeCustomFloatFromBool(Bool=$System.IsSynced, True Float=666, False Float=555)
   )
   &InitialMeshOrientation(
    Mesh Axis To Orient=<0 0 1>
    Mesh Orientation Mode=::NewEnumerator2
    Orientation Vector=$Particles.ShapeLocation.ShapeNormal
    Use Rotation Vector=false
   )
  ]
  particleUpdateScript [
   &ParticleState()
   &UpdateMeshOrientation(
    Rotation Rate=&RandomRangeFloat(Minimum=2, Maximum=4)
    Rotation Vector=&RandomVector(Vector Scale=1)
   )
   &ScaleMeshSize(Scale Factor=&VectorFromFloat(Value=&FloatFromCurve(FloatCurve=DataInterface() {propertyValues {Curve {keys [
          {interpMode=RCIM_Cubic, tangentMode=RCTM_User, time=0, value=1}
          {arriveTangent=-1.6697, interpMode=RCIM_Cubic, leaveTangent=-1.6697, tangentMode=RCTM_User, time=1, value=0}
         ]}}})))
   &Drag(Drag=&RandomRangeFloat(Minimum=6, Maximum=8), Use Linear Drag=true, Use Rotational Drag=true)
   &CurlNoiseForce(Noise Strength=111, Pan Noise=true, Pan Noise Field=<0 0 0.1>)
   &GravityForce()
   &SolveForcesAndVelocity(Clamp Velocity=false, Limit Acceleration=false)
  ]
  rendererClasses [&NiagaraMeshRendererProperties]
  renderers [{rendererClass=&NiagaraMeshRendererProperties, rendererIndex=0}]
 }
 SnowAirStreaks {
  emitterUpdateScript [
   &EmitterState(Life Cycle Mode=::NewEnumerator1, Loop Behavior=::NewEnumerator1, Loop Delay=0.1, Loop Duration Mode=::NewEnumerator1, UseLoopDelay=true)
   &SpawnRate(
    SpawnRate=&MakeCustomFloatFromBool(Bool=$System.IsSynced, True Float=20, False Float=15)
    Use Spawn Probability=false
   )
  ]
  particleSpawnScript [
   &FN_SetSpawnGroup(Number of Spawn Groups=2)
   /SetVariables(Particles.RandomFloat=&UniformRangedFloat(Minimum=0, Maximum=3)) {isSetParameters=true}
   &InitializeParticle(
    Color Maximum=<rgba 0.3 0.4551 0.6 0.8>
    Color Minimum=<rgba 0.385 0.5215 0.7 0.6>
    Color Mode=::NewEnumerator2
    Lifetime=&MakeCustomFloatFromBool(Bool=$System.IsSynced, True Float=&RandomRangeFloat(Minimum=0.8, Maximum=1.2), False Float=0.75)
    Sprite Size Max=<70 60>
    Sprite Size Min=<40 60>
    Sprite Size Mode=::NewEnumerator2
    UsePositionOffset=false
    Write Material Random=true
   ) {module=InitializeParticle001}
   &AddVelocity(
    Cone Angle=&MakeCustomFloatFromBool(Bool=$System.IsSynced, True Float=30, False Float=60)
    Cone Axis=&MakeCustomVectorFromBool(Bool=$System.IsSynced, True Vector=$System.SnowDirection, False Vector=<0 0 1>)
    Module.UseConeFalloff=false
    Module.UseVelDistribution=false
    Override Seed=false
    Velocity Mode=::NewEnumerator2
    Velocity Speed=&RandomRangeFloat(Minimum=200, Maximum=400)
   ) {module=AddVelocity001}
   &SubUVAnimation(End Frame=3, SubUV Animation Mode:ENiagara_SubUVLookupMode=::NewEnumerator2)
   &CameraOffset(Camera Offset Amount=25)
  ]
  particleUpdateScript [
   &ParticleState()
   &Drag(Drag=6, Use Linear Drag=true, Use Rotational Drag=false)
   &ScaleVelocity(Velocity Scale=&VectorFromFloat(Value=&FloatFromCurve(FloatCurve=DataInterface() {propertyValues {Curve {keys [
          {arriveTangent=-3.6709, interpMode=RCIM_Cubic, leaveTangent=-3.6709, tangentMode=RCTM_User, time=0, value=1}
          {interpMode=RCIM_Cubic, tangentMode=RCTM_User, time=1, value=0}
         ]}}})))
   &SolveForcesAndVelocity(Clamp Velocity=false, Limit Acceleration=false)
   &ScaleSpriteSize(Scale Factor=&Vector2DFromCurve(
     Vector2Curve=DataInterface() {propertyValues {
       XCurve {keys [
         {arriveTangent=2.5428, interpMode=RCIM_Cubic, leaveTangent=2.5428, tangentMode=RCTM_User, time=0, value=1}
         {arriveTangent=0.1281, interpMode=RCIM_Cubic, leaveTangent=0.1281, tangentMode=RCTM_User, time=1, value=2}
        ]}
       YCurve {keys [{arriveTangent=2.8823, time=0, value=1}, {arriveTangent=2.0297, leaveTangent=2.0297, time=1, value=0.5}]}
      }}
     ScaleCurve=&Multiply_Vector2D_ByFloat(
      Float=&RandomRangeFloat(Minimum=0.6, Maximum=1)
      Vector2D=&RandomRangeVector2D(Minimum=<0.6 0.75>, Maximum=<0.8 1>)
     )
    ))
   &ScaleColor(
    Scale Alpha=&FloatFromCurve(
     FloatCurve=DataInterface() {propertyValues {Curve {keys [
         {interpMode=RCIM_Cubic, leaveTangent=-4.5779, tangentMode=RCTM_Break, time=0, value=1.5}
         {interpMode=RCIM_Cubic, tangentMode=RCTM_Break, time=1, value=0}
        ]}}}
     Scale Curve=&Multiply_Float(
      A=&Multiply_Float(
       A=&FNTimeOfDayScalar(Night Scale=0.25, Morning Scale=0.5, Evening Scale=0.4)
       B=&RandomRangeFloat(Minimum=1, Maximum=1.25)
      )
      B=&MakeCustomFloatFromBool(Bool=$System.IsSynced, True Float=0.4, False Float=0.75)
     )
    )
    Scale RGB=&Multiply_Vector(
     A=&VectorFromCurve(
      VectorCurve=DataInterface() {propertyValues {
        XCurve {keys [{arriveTangent=-5.0711, interpMode=RCIM_Cubic, leaveTangent=-5.0711, tangentMode=RCTM_User, time=0, value=2}, <at 1 0.5>]}
        YCurve {keys [{arriveTangent=-5.0711, interpMode=RCIM_Cubic, leaveTangent=-5.0711, tangentMode=RCTM_User, time=0, value=2}, <at 1 0.5>]}
        ZCurve {keys [{arriveTangent=-5.0711, interpMode=RCIM_Cubic, leaveTangent=-5.0711, tangentMode=RCTM_User, time=0, value=2}, <at 1 0.5>]}
       }}
      Scale Curve=&VectorFromFloat(Value=&MakeCustomFloatFromBool(Bool=$System.IsSynced, True Float=2, False Float=0.75))
     )
     B=&MakeVectorFromLinearColorRGB(Color=&FN_Select_Color_from_Time_Of_Day(Morning=<rgba 1 0.8463 0.7189 1>, Day=<rgba 1 1 1 1>, Evening=<rgba 1 0.8371 0.7936 1>, Night=<rgba 0.6111 0.6842 1 1>))
    )
    ScaleA=true
    ScaleRGB=true
   )
   &DynamicMaterialParameters(
    Index 0 Param 1=20
    Index 0 Param 2=1
    Index 0 Param 3=1
    Index 0 Param 4=&FloatFromCurve(FloatCurve=DataInterface() {propertyValues {Curve {keys [
         {interpMode=RCIM_Cubic, tangentMode=RCTM_User, time=0, value=0}
         {arriveTangent=2.7922, interpMode=RCIM_Cubic, leaveTangent=2.7922, tangentMode=RCTM_User, time=1, value=1}
        ]}}})
    Index 2 Param 1=2
    Index 2 Param 2=&FloatFromCurve(
     FloatCurve=DataInterface() {propertyValues {Curve {keys [
         {arriveTangent=0.0308, interpMode=RCIM_Cubic, leaveTangent=0.0308, tangentMode=RCTM_User, time=0, value=1}
         {arriveTangent=-2.7663, interpMode=RCIM_Cubic, leaveTangent=-2.7663, tangentMode=RCTM_User, time=1, value=0}
        ]}}}
     Scale Curve=&MakeCustomFloatFromBool(Bool=$System.IsSynced, True Float=0.775, False Float=0.75)
    )
    Module.Param0WriteEnabled=true
    Module.Param0WriteEnabled2=true
    Module.Param1WriteEnabled=true
    Module.Param1WriteEnabled2=true
    Module.Param2WriteEnabled=true
    Module.Param2WriteEnabled2=true
    Module.Param3WriteEnabled=true
    Module.Param3WriteEnabled2=true
    Write Parameter Index 2=true
   )
  ]
  rendererClasses [&NiagaraSpriteRendererProperties]
  renderers [{rendererClass=&NiagaraSpriteRendererProperties}]
 }
 Snow_Powder {
  emitterUpdateScript [
   &EmitterState()
   &SpawnRate(
    SpawnRate=&MakeCustomFloatFromBool(Bool=$System.IsSynced, True Float=5, False Float=4)
    Use Spawn Probability=false
   )
  ]
  particleSpawnScript [
   &InitializeParticle(
    Color=<rgba 0.4492 0.6108 0.8177 0.3>
    Lifetime=&MakeCustomFloatFromBool(Bool=$System.IsSynced, True Float=1.2, False Float=1.6)
    Sprite Rotation Mode=::NewEnumerator2
    Sprite Size Mode=::NewEnumerator1
    Uniform Sprite Size Max=133
    Uniform Sprite Size Min=99
    UsePositionOffset=false
    Write Material Random=true
   )
   &AddVelocity(
    Cone Angle=360
    Cone Axis=<0 1 1>
    Module.UseConeFalloff=true
    Module.UseVelDistribution=true
    Override Seed=false
    Speed Falloff From Cone Axis=0.9
    Velocity Mode=::NewEnumerator2
    Velocity Speed=150
   )
  ]
  particleUpdateScript [
   &DynamicMaterialParameters(
    Index 0 Param 1=3
    Index 0 Param 2=&FloatFromCurve(FloatCurve=DataInterface() {propertyValues {Curve {keys [
         {arriveTangent=9.0725, interpMode=RCIM_Cubic, leaveTangent=9.0725, tangentMode=RCTM_User, time=0, value=0}
         {arriveTangent=-2.5249, interpMode=RCIM_Cubic, leaveTangent=-2.5249, tangentMode=RCTM_User, time=0.1408, value=0.9847}
         {arriveTangent=-0.0405, interpMode=RCIM_Cubic, leaveTangent=-0.0405, tangentMode=RCTM_User, time=1, value=0}
        ]}}})
    Module.Index0Param0Write=true
    Module.Index0Param1Write=true
    Module.Index0Param2Write=true
    Module.Index0Param3Write=true
   )
   &CameraOffset(Camera Offset Amount=&FloatFromCurve(
     FloatCurve=DataInterface() {propertyValues {Curve {keys [
         {arriveTangent=4.1785, interpMode=RCIM_Cubic, leaveTangent=4.1785, tangentMode=RCTM_User, time=0, value=0}
         {arriveTangent=1.1179, interpMode=RCIM_Cubic, leaveTangent=1.1179, tangentMode=RCTM_User, time=0.2042, value=0.6228}
         {arriveTangent=0.0224, interpMode=RCIM_Cubic, leaveTangent=0.0224, tangentMode=RCTM_User, time=1, value=1}
        ]}}}
     Scale Curve=55
    ))
   &Drag(Drag=2, Use Linear Drag=true, Use Rotational Drag=true)
   &SolveForcesAndVelocity(Clamp Velocity=false, Limit Acceleration=false)
   &ParticleState()
   &ScaleColor(
    Scale Alpha=&Multiply_Float(
     A=&FloatFromCurve(
      FloatCurve=DataInterface() {propertyValues {Curve {keys [{interpMode=RCIM_Cubic, tangentMode=RCTM_User, time=0, value=1}, {interpMode=RCIM_Cubic, tangentMode=RCTM_User, time=1, value=0}]}}}
      Scale Curve=&RandomRangeFloat(Minimum=0.5, Maximum=1)
     )
     B=&MakeCustomFloatFromBool(Bool=$System.IsSynced, True Float=1, False Float=0.6)
    )
    Scale RGB=&MakeVectorFromLinearColorRGB(Color=$System.TODColorMultiplier)
    ScaleA=true
    ScaleRGB=true
   )
   &SubUVAnimation(
    "Sprite Renderer "=DataInterface() {propertyValues {SpriteRenderer=&self:Snow_Powder_4.NiagaraSpriteRendererProperties_0}}
    UseEndFrame=false
    UseStartFrame=false
   )
   &ScaleSpriteSize(Uniform Curve Sprite Scale=DataInterface() {propertyValues {Curve {keys [
        {arriveTangent=4.1785, interpMode=RCIM_Cubic, leaveTangent=4.1785, tangentMode=RCTM_User, time=0, value=0}
        {arriveTangent=0.9995, interpMode=RCIM_Cubic, leaveTangent=0.9995, tangentMode=RCTM_User, time=0.2223, value=0.6825}
        {arriveTangent=-0.0405, interpMode=RCIM_Cubic, leaveTangent=-0.0405, tangentMode=RCTM_User, time=1, value=1}
       ]}}})
  ]
  rendererClasses [&NiagaraSpriteRendererProperties]
  renderers [{rendererClass=&NiagaraSpriteRendererProperties}]
 }
 SnowflakesCPU {
  emitterUpdateScript [
   &EmitterState()
   &SpawnRate(
    SpawnRate=&Multiply_Float(
     A=&MakeCustomFloatFromBool(Bool=$System.IsSynced, True Float=&RandomRangeFloat(Minimum=11, Maximum=22), False Float=3)
     B=1
    )
    Use Spawn Probability=false
   )
  ]
  particleSpawnScript [
   &InitializeParticle(
    Color Maximum=<rgba 0.42 0.4712 0.6 0.6>
    Color Minimum=<rgba 0.3882 0.5623 1 0.9>
    Color Mode=::NewEnumerator2
    Lifetime Max=4
    Lifetime Min=3
    Lifetime Mode=::NewEnumerator1
    Sprite Rotation Mode=::NewEnumerator2
    Sprite Size Mode=::NewEnumerator1
    Uniform Sprite Size Max=5
    Uniform Sprite Size Min=3
    UsePositionOffset=false
    Write Material Random=true
   )
   &ShapeLocation(MaskSpawnGroup=false, Override Seed=false, Sphere Radius=1, UseOwnerScale=true)
   &AddVelocity(
    Cone Angle=&MakeCustomFloatFromBool(Bool=$System.IsSynced, True Float=&RandomRangeFloat(Minimum=60, Maximum=180), False Float=180)
    Cone Axis=&MakeCustomVectorFromBool(Bool=$System.IsSynced, True Vector=$System.SnowDirection, False Vector=<0 -0.2 1>)
    Distribution Along Cone Axis=0.75
    Inner Cone Angle=5
    Module.UseConeFalloff=true
    Module.UseVelDistribution=true
    Override Seed=false
    Speed Falloff From Cone Axis=0.3
    Velocity Mode=::NewEnumerator2
    Velocity Speed=&MakeCustomFloatFromBool(Bool=$System.IsSynced, True Float=&RandomRangeFloat(Minimum=444, Maximum=666), False Float=555)
   )
   &SubUVAnimation(
    Evaluation Type=::NewEnumerator0
    "Sprite Renderer "=DataInterface() {propertyValues {SpriteRenderer=&self:SnowflakesCPU_5.NiagaraSpriteRendererProperties_1}}
    SubUV Animation Mode:ENiagara_SubUVLookupModeV2=::NewEnumerator2
    UseEndFrame=false
    UseStartFrame=false
   )
   &InheritVelocity(Speed Limit=false, UseSpeedThreshold=false)
  ]
  particleUpdateScript [
   &ParticleState()
   &Drag(Drag=&RandomRangeFloat(Minimum=4, Maximum=6), Use Linear Drag=true, Use Rotational Drag=true)
   &GravityForce(Gravity=&RandomRangeVector(Minimum=<0 0 -88>, Maximum=<1 1 -111>))
   &AccelerationForce(
    Acceleration=&MakeCustomVectorFromBool(
     Bool=$System.IsSynced
     True Vector=&Multiply_VectorByFloat(
      Vector=&Multiply_Vector(A=$StackContext.ShapeLocation.ShapeNormal, B=<1 1 0>)
      Float=&Lerp_Float(
       A=333
       B=33
       Alpha=&FloatFromCurve(FloatCurve=DataInterface() {propertyValues {Curve {keys [{tangentMode=RCTM_User, time=0, value=0}, {tangentMode=RCTM_User, time=1, value=1}]}}})
      )
     )
     False Vector=<0 0 0>
    )
    Coordinate Space=::Local
   )
   &InheritSourceMovement(ApplyForce=true, ApplyForceFalloff=false, ApplyPosition=false, Module.LimitForceSpeed=false)
   &SolveForcesAndVelocity(Clamp Velocity=false, Limit Acceleration=false)
   &SpriteRotationRate(Rotation Rate=&FloatFromCurve(
     FloatCurve=DataInterface() {propertyValues {Curve {keys [
         {arriveTangent=4.1785, interpMode=RCIM_Cubic, tangentMode=RCTM_Break, time=0, value=0}
         {arriveTangent=0.0224, interpMode=RCIM_Cubic, leaveTangent=0.0224, tangentMode=RCTM_Break, time=0.3, value=1}
        ]}}}
     CurveIndex=$Particles.Age
     Scale Curve=&RandomRangeFloat(Minimum=-360, Maximum=360)
    ))
   &ScaleSpriteSize(
    Scale Factor=&Multiply_Vector2D_ByFloat(
     Float=&FloatFromCurve(
      FloatCurve=DataInterface() {propertyValues {Curve {keys [
          {tangentMode=RCTM_User, time=0, value=1}
          {arriveTangent=-0.0405, interpMode=RCIM_Cubic, leaveTangent=-0.0405, tangentMode=RCTM_User, time=0.7009, value=1}
          {arriveTangent=-8.0089, interpMode=RCIM_Cubic, leaveTangent=-8.0089, tangentMode=RCTM_User, time=1, value=0}
         ]}}}
      Scale Curve=1.75
     )
     Vector2D=&Lerp_Vector2(
      A=<1 2>
      B=&RandomRangeVector2D(Minimum=<0.8 0.8>, Maximum=<1 1>)
      Alpha=&FloatFromCurve(
       FloatCurve=DataInterface() {propertyValues {Curve {keys [
           {arriveTangent=4.1785, interpMode=RCIM_Cubic, tangentMode=RCTM_Break, time=0, value=0}
           {arriveTangent=9.9282, interpMode=RCIM_Cubic, leaveTangent=0.0224, tangentMode=RCTM_Break, time=0.3, value=1}
          ]}}}
       CurveIndex=$Particles.Age
      )
     )
    )
    Scale Sprite Size Mode=::NewEnumerator2
   )
   &SpriteFacingAndAlignment(
    Sprite Alignment=&Lerp_Vector(
     A=&ReturnCameraVector(Camera Vectors=::NewEnumerator2, Coordinate Space=::Simulation)
     B=&NormalizeVector(Vector To Normalize=$Particles.Velocity, Safe Normalize=true, Safe Fallback Threshold=0, Safe Fallback Vector=<0 0 0.001>)
     Alpha=&ClampFloat(Float=&Divide_Float(A=&MakeFloatFromVector(Vector=$Particles.Velocity, Channel=::NewEnumerator2), B=33))
    )
    WriteAlignment=true
    WriteFacing=false
   )
   &ScaleColor(
    Scale Alpha=&Multiply_Float(
     A=&FloatFromCurve(
      FloatCurve=DataInterface() {propertyValues {Curve {keys [
          {interpMode=RCIM_Cubic, leaveTangent=2.9196, tangentMode=RCTM_Break, time=0, value=0.25}
          {arriveTangent=-0.0405, interpMode=RCIM_Cubic, leaveTangent=-0.0405, tangentMode=RCTM_User, time=0.7009, value=1}
          {arriveTangent=-8.0089, interpMode=RCIM_Cubic, leaveTangent=-8.0089, tangentMode=RCTM_User, time=1, value=0}
         ]}}}
      Scale Curve=&RandomRangeFloat(Minimum=0.6, Maximum=1)
     )
     B=1
    )
    Scale RGB=&MakeVectorFromLinearColorRGB(Color=$System.TODColorMultiplier)
    ScaleA=true
    ScaleRGB=true
   )
   /SetVariables(
    Particles.Sparkles Color:LinearColor=&MakeLinearColorFromVectorAndFloat(
     "Vector (RGB)"=&MakeVectorFromLinearColorRGB(Color=&MultiplyLinearColor(
       Linear Color=&MultiplyLinearColor(
        Linear Color=$Particles.Color
        Scale Factor=&MakeVector4_FromLinearColor(LinearColor=&FN_Select_Color_from_Time_Of_Day(Morning=<rgba 1 0.8463 0.7189 1>, Day=<rgba 1 1 1 1>, Evening=<rgba 1 0.8371 0.7936 1>, Night=<rgba 0.1429 0.183 0.4497 1>))
       )
       Scale Factor=&MakeFloat4FromVectorAndFloat(
        "Vector (RGB)"=&VectorFromFloat(Value=&RandomRangeFloat(Minimum=4, Maximum=11))
        "Float (Alpha)"=1
       )
      ))
     "Float (Alpha)"=&Lerp_Float(
      A=&Multiply_Float(
       A=&FN_Select_Float_from_Time_Of_Day(Morning=2, Day=2, Evening=2, Night=10)
       B=&Waveform(
        Global Phase=$Particles.Age
        "Waveform [1]"=::NewEnumerator0
        "[1] Phase"=$Particles.Age
        "[1] Frequency"=2
        "[1] Amplitude Min/Max"=&MakeVector2D(X=0, Y=&RandomRangeFloat(Minimum=-1, Maximum=1))
        "[1] Scale"=&RandomRangeFloat(Minimum=-1, Maximum=1)
       )
      )
      B=0
      Alpha=&NormalizeDistanceRange(
       Start Position:NiagaraPosition=&TransformVector(
        Vector=&AddVectorToPosition(Position=$Engine.Emitter.SimulationPosition, Vector=<0 100 50>)
        Destination Space=::Local
       )
       End Position:NiagaraPosition=$Particles.Position
       Distance:NiagaraFloat=50
       Invert Normalized Range:NiagaraBool=false
      )
     )
    )
    Particles.SpriteSizeSparkles:Vector2f=&Multiply_Vector2D_ByFloat(Float=&RandomRangeFloat(Minimum=0.5, Maximum=1.5), Vector2D=$Particles.SpriteSize)
   ) {isSetParameters=true, module=SetVariables_#5}
  ]
  rendererClasses [&NiagaraSpriteRendererProperties]
  renderers [
   {rendererClass=&NiagaraSpriteRendererProperties}
   {rendererClass=&NiagaraSpriteRendererProperties, rendererIndex=1}
  ]
 }
 SnowflakesGPU {
  simTarget=GPUComputeSim
  emitterUpdateScript [
   &EmitterState()
   &SpawnBurst_Instantaneous(Spawn Count=6, Use Spawn Probability=false, UseLoopCountLimit=false)
   &SpawnRate(
    SpawnRate=&Multiply_Float(
     A=&MakeCustomFloatFromBool(
      Bool=$System.IsSynced
      True Float=&RandomRangeFloat(Minimum=44, Maximum=88, Evaluation Type=::NewEnumerator1)
      False Float=3
     )
     B=2
    )
    Use Spawn Probability=false
   )
  ]
  particleSpawnScript [
   &InitializeParticle(
    Color Maximum=<rgba 0.42 0.4712 0.6 0.6>
    Color Minimum=<rgba 0.3882 0.5623 1 0.9>
    Color Mode=::NewEnumerator2
    Lifetime Max=4
    Lifetime Min=3
    Lifetime Mode=::NewEnumerator1
    Sprite Rotation Mode=::NewEnumerator2
    Sprite Size Mode=::NewEnumerator1
    Uniform Sprite Size Max=3
    Uniform Sprite Size Min=1
    UsePositionOffset=false
    Write Material Random=true
   )
   &ShapeLocation(MaskSpawnGroup=false, Override Seed=false, Sphere Radius=1, UseOwnerScale=true)
   &AddVelocity(
    Cone Angle=&MakeCustomFloatFromBool(Bool=$System.IsSynced, True Float=&RandomRangeFloat(Minimum=60, Maximum=180), False Float=180)
    Cone Axis=&MakeCustomVectorFromBool(Bool=$System.IsSynced, True Vector=$System.SnowDirection, False Vector=<0 -0.2 1>)
    Distribution Along Cone Axis=0.75
    Inner Cone Angle=5
    Module.UseConeFalloff=true
    Module.UseVelDistribution=true
    Override Seed=false
    Speed Falloff From Cone Axis=0.3
    Velocity Mode=::NewEnumerator2
    Velocity Speed=&MakeCustomFloatFromBool(Bool=$System.IsSynced, True Float=&RandomRangeFloat(Minimum=444, Maximum=666), False Float=555)
   )
   &SubUVAnimation(
    Evaluation Type=::NewEnumerator0
    "Sprite Renderer "=DataInterface() {propertyValues {SpriteRenderer=&self:SnowflakesGPU_6.NiagaraSpriteRendererProperties_1}}
    SubUV Animation Mode:ENiagara_SubUVLookupModeV2=::NewEnumerator2
    UseEndFrame=false
    UseStartFrame=false
   )
   &InheritVelocity(Speed Limit=false, UseSpeedThreshold=false)
  ]
  particleUpdateScript [
   &ParticleState()
   &Drag(Drag=&RandomRangeFloat(Minimum=4, Maximum=6), Use Linear Drag=true, Use Rotational Drag=true)
   &GravityForce(Gravity=&RandomRangeVector(Minimum=<0 0 -88>, Maximum=<1 1 -111>))
   &CurlNoiseForce(
    Noise Quality / Cost=::NewEnumerator0
    Noise Strength=&FloatFromCurve(
     FloatCurve=DataInterface() {propertyValues {Curve {keys [
         {arriveTangent=4.1785, interpMode=RCIM_Cubic, leaveTangent=4.1785, tangentMode=RCTM_User, time=0, value=0}
         {arriveTangent=1.1179, interpMode=RCIM_Cubic, leaveTangent=1.1179, tangentMode=RCTM_User, time=0.2042, value=0.6228}
         {arriveTangent=0.0224, interpMode=RCIM_Cubic, leaveTangent=0.0224, tangentMode=RCTM_User, time=1, value=1}
        ]}}}
     Scale Curve=333
    )
    Pan Noise=false
   )
   &AccelerationForce(
    Acceleration=&MakeCustomVectorFromBool(
     Bool=$System.IsSynced
     True Vector=&Multiply_VectorByFloat(
      Vector=&Multiply_Vector(A=$StackContext.ShapeLocation.ShapeNormal, B=<1 1 0>)
      Float=&Lerp_Float(
       A=333
       B=33
       Alpha=&FloatFromCurve(FloatCurve=DataInterface() {propertyValues {Curve {keys [{tangentMode=RCTM_User, time=0, value=0}, {tangentMode=RCTM_User, time=1, value=1}]}}})
      )
     )
     False Vector=<0 0 0>
    )
    Coordinate Space=::Local
   )
   &InheritSourceMovement(ApplyForce=true, ApplyForceFalloff=false, ApplyPosition=false, Module.LimitForceSpeed=false)
   &SolveForcesAndVelocity(Clamp Velocity=false, Limit Acceleration=false)
   &SpriteRotationRate(Rotation Rate=&FloatFromCurve(
     FloatCurve=DataInterface() {propertyValues {Curve {keys [
         {arriveTangent=4.1785, interpMode=RCIM_Cubic, tangentMode=RCTM_Break, time=0, value=0}
         {arriveTangent=0.0224, interpMode=RCIM_Cubic, leaveTangent=0.0224, tangentMode=RCTM_Break, time=0.3, value=1}
        ]}}}
     CurveIndex=$Particles.Age
     Scale Curve=&RandomRangeFloat(Minimum=-360, Maximum=360)
    ))
   &ScaleSpriteSize(
    Scale Factor=&Multiply_Vector2D_ByFloat(
     Float=&FloatFromCurve(
      FloatCurve=DataInterface() {propertyValues {Curve {keys [
          {tangentMode=RCTM_User, time=0, value=1}
          {arriveTangent=-0.0405, interpMode=RCIM_Cubic, leaveTangent=-0.0405, tangentMode=RCTM_User, time=0.7009, value=1}
          {arriveTangent=-8.0089, interpMode=RCIM_Cubic, leaveTangent=-8.0089, tangentMode=RCTM_User, time=1, value=0}
         ]}}}
      Scale Curve=1.75
     )
     Vector2D=&Lerp_Vector2(
      A=<1 2>
      B=&RandomRangeVector2D(Minimum=<0.8 0.8>, Maximum=<1 1>)
      Alpha=&FloatFromCurve(
       FloatCurve=DataInterface() {propertyValues {Curve {keys [
           {arriveTangent=4.1785, interpMode=RCIM_Cubic, tangentMode=RCTM_Break, time=0, value=0}
           {arriveTangent=9.9282, interpMode=RCIM_Cubic, leaveTangent=0.0224, tangentMode=RCTM_Break, time=0.3, value=1}
          ]}}}
       CurveIndex=$Particles.Age
      )
     )
    )
    Scale Sprite Size Mode=::NewEnumerator2
   )
   &SpriteFacingAndAlignment(
    Sprite Alignment=&Lerp_Vector(
     A=&ReturnCameraVector(Camera Vectors=::NewEnumerator2, Coordinate Space=::Simulation)
     B=&NormalizeVector(Vector To Normalize=$Particles.Velocity, Safe Normalize=true, Safe Fallback Threshold=0, Safe Fallback Vector=<0 0 0.001>)
     Alpha=&ClampFloat(Float=&Divide_Float(A=&MakeFloatFromVector(Vector=$Particles.Velocity, Channel=::NewEnumerator2), B=33))
    )
    WriteAlignment=true
    WriteFacing=false
   )
   &ScaleColor(
    Scale Alpha=&Multiply_Float(
     A=&FloatFromCurve(
      FloatCurve=DataInterface() {propertyValues {Curve {keys [
          {interpMode=RCIM_Cubic, leaveTangent=2.9196, tangentMode=RCTM_Break, time=0, value=0.25}
          {arriveTangent=-0.0405, interpMode=RCIM_Cubic, leaveTangent=-0.0405, tangentMode=RCTM_User, time=0.7009, value=1}
          {arriveTangent=-8.0089, interpMode=RCIM_Cubic, leaveTangent=-8.0089, tangentMode=RCTM_User, time=1, value=0}
         ]}}}
      Scale Curve=&RandomRangeFloat(Minimum=0.6, Maximum=1)
     )
     B=1
    )
    Scale RGB=&MakeVectorFromLinearColorRGB(Color=$System.TODColorMultiplier)
    ScaleA=true
    ScaleRGB=true
   )
   /SetVariables(
    Particles.Sparkles Color:LinearColor=&MakeLinearColorFromVectorAndFloat(
     "Vector (RGB)"=&MakeVectorFromLinearColorRGB(Color=&MultiplyLinearColor(
       Linear Color=&MultiplyLinearColor(
        Linear Color=$Particles.Color
        Scale Factor=&MakeVector4_FromLinearColor(LinearColor=&FN_Select_Color_from_Time_Of_Day(Morning=<rgba 1 0.8463 0.7189 1>, Day=<rgba 1 1 1 1>, Evening=<rgba 1 0.8371 0.7936 1>, Night=<rgba 0.1429 0.183 0.4497 1>))
       )
       Scale Factor=&MakeFloat4FromVectorAndFloat(
        "Vector (RGB)"=&VectorFromFloat(Value=&RandomRangeFloat(Minimum=4, Maximum=11))
        "Float (Alpha)"=1
       )
      ))
     "Float (Alpha)"=&Lerp_Float(
      A=&Multiply_Float(
       A=&FN_Select_Float_from_Time_Of_Day(Morning=2, Day=2, Evening=2, Night=10)
       B=&Waveform(
        Global Phase=$Particles.Age
        "Waveform [1]"=::NewEnumerator0
        "[1] Phase"=$Particles.Age
        "[1] Frequency"=2
        "[1] Amplitude Min/Max"=&MakeVector2D(X=0, Y=&RandomRangeFloat(Minimum=-1, Maximum=1))
        "[1] Scale"=&RandomRangeFloat(Minimum=-1, Maximum=1)
       )
      )
      B=0
      Alpha=&NormalizeDistanceRange(
       Start Position:NiagaraPosition=&TransformVector(
        Vector=&AddVectorToPosition(Position=$Engine.Emitter.SimulationPosition, Vector=<0 100 50>)
        Destination Space=::Local
       )
       End Position:NiagaraPosition=$Particles.Position
       Distance:NiagaraFloat=50
       Invert Normalized Range:NiagaraBool=false
      )
     )
    )
    Particles.SpriteSizeSparkles:Vector2f=&Multiply_Vector2D_ByFloat(Float=&RandomRangeFloat(Minimum=0.5, Maximum=1.5), Vector2D=$Particles.SpriteSize)
   ) {isSetParameters=true, module=SetVariables_#5}
  ]
  rendererClasses [&NiagaraSpriteRendererProperties]
  renderers [
   {rendererClass=&NiagaraSpriteRendererProperties}
   {rendererClass=&NiagaraSpriteRendererProperties, rendererIndex=1}
  ]
 }
}
summary {
 systemName=NS_Emote_ShrimpStroll
 emitters [
  {emitterName=Snow_Powder, rendererClasses [&NiagaraSpriteRendererProperties]}
  {emitterName=IceChunks, rendererClasses [&NiagaraMeshRendererProperties]}
  {emitterName=SnowflakesGPU, simTarget=GPUComputeSim, rendererClasses [&NiagaraSpriteRendererProperties]}
  {emitterName=SnowAirStreaks, rendererClasses [&NiagaraSpriteRendererProperties]}
  {emitterName=SnowflakesCPU, rendererClasses [&NiagaraSpriteRendererProperties]}
 ]
 userVariables [
  User.bIsFrontEnd {type=NiagaraBool}
  User.bIsFrontEndPreview {type=NiagaraBool}
  User.NormalizedNotifyProgress {type=NiagaraFloat, defaultValue=0, description=""}
 ]
}
